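; Spotlight prop Blueprint that dresses a light fixture mesh with an adjustable light shaft and flare effect.
; The construction script assigns a WPO material to the fixture, runs the parent script, then pushes rotation to the shaft material and the spotlight color to the fixture emissive and flare materials.
; Beam width and height drive the shaft mesh's relative scale.
; The event graph holds no active logic.
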(event EventBeginPlay
  (|CallParentFunction) ; This node is disabled and will not be called.\nDrag off pins to build functionality.
)

(event Collision|EventActorBeginOverlap
  (params (OtherActor :type Actor%20Object%20Reference)))

(event EventTick
  (params (DeltaSeconds :type Float%20%28single-precision%29)))

(fn ConstructionScript
  (Rendering|Material|SetMaterial (Variables|Default|GetS_Mother_Spotlight_01) 0 "/Clyde/Environments/Props/LightSource/Materials/M_Clyde_LC_SpotLight_WPO.M_Clyde_LC_SpotLight_WPO")
  (|CallParentFunction)
  (Rendering|Material|SetScalarParameterValueonMaterials (Variables|Default|GetEditorOnly_sm_neo_tilted_light_shaft) "Rotation" (Variables|SpotlightFXSettings|GetRotation))
  (Rendering|Material|SetVectorParameterValueonMaterials (Variables|Default|GetS_Mother_Spotlight_01) "EmissiveColorMultiplier" (Math|Conversions|ToVector(LinearColor) (Variables|SpotlightFXSettings|GetSpotlightColor)))
  (Rendering|Material|SetVectorParameterValueonMaterials (Variables|Default|GetEditorOnly_sm_neo_tilted_light_shaft_Flares) "Color" (Math|Conversions|ToVector(LinearColor) (Variables|SpotlightFXSettings|GetSpotlightColor)))
  (Transformation|SetRelativeScale3D (Variables|Default|GetEditorOnly_sm_neo_tilted_light_shaft) (Math|Vector|MakeVector (Variables|SpotlightFXSettings|GetBeam%20Width) (Variables|SpotlightFXSettings|GetBeam%20Width) (Variables|SpotlightFXSettings|GetBeam%20Height))))
