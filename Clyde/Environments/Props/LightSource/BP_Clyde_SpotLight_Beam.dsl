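; Spotlight beam prop, a child Blueprint that adds a tilted light shaft mesh and flares to its parent spotlight.
; Its setSpotlightProperties override calls the parent, then scales the shaft from the beam width and height variables.
; It also pushes the spotlight color to the flare materials and the Rotation variable to a shaft material parameter.
; BeginPlay, Tick and overlap events are empty, and construction only calls the parent.
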(event EventBeginPlay)

(event Collision|EventActorBeginOverlap
  (params (OtherActor :type Actor%20Object%20Reference)))

(event EventTick
  (params (DeltaSeconds :type Float%20%28single-precision%29)))

(event EventsetSpotlightProperties
  (params (Rotation%20Max :type Float%20%28double-precision%29) (Rotation%20Speed :type Float%20%28double-precision%29) (Beam%20Height :type Float%20%28double-precision%29) (Beam%20Width :type Float%20%28double-precision%29) (SpotlightColor :type Linear%20Color%20Structure))
  (|CallParentFunction Rotation%20Max Rotation%20Speed Beam%20Height Beam%20Width SpotlightColor)
  (Transformation|SetRelativeScale3D (Variables|Default|Getsm_neo_tilted_light_shaft) (Math|Vector|MakeVector (Variables|SpotlightFXSettings|GetBeam%20Width) (Variables|SpotlightFXSettings|GetBeam%20Width) (Variables|SpotlightFXSettings|GetBeam%20Height)))
  (Rendering|Material|SetVectorParameterValueonMaterials (Variables|Default|Getsm_neo_tilted_light_shaft_Flares) "Color" (Math|Conversions|ToVector(LinearColor) (Variables|SpotlightFXSettings|GetSpotlightColor)))
  (Rendering|Material|SetScalarParameterValueonMaterials (Variables|Default|Getsm_neo_tilted_light_shaft) "Rotation" (Variables|SpotlightFXSettings|GetRotation)))

(fn ConstructionScript
  (|CallParentFunction))
