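# Lit material for a Neo tilted spotlight prop that sways the light in the vertex shader.
# A time-driven sine with speed, offset and amplitude parameters rotates vertices around an adjustable axis pivoted at actor position plus an offset, masked by red vertex color.
# FixRotateAboutAxisNormals corrects normals for the rotation, which then blend with the normal map.
# A packed mask drives metallic, specular, roughness and tinted emissive, and MF_CreativeMoveTool wraps the final attributes.
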Material /Clyde/Environments/Props/LightSource/Materials/M_Clyde_LC_SpotLight_WPO
graph {
 properties {bUseMaterialAttributes=true, numCustomizedUVs=5, pixelDepthOffsetMode=PDOM_Legacy, usedShadingModels=Default Lit}
 nodes [
  AppendVector(A=TextureCoordinate() {coordinateIndex=3}, B=ComponentMask(None=TextureCoordinate() {coordinateIndex=4}) {r=true})
  n1 = ScalarParameter() {parameterName=Rotation}
  Frac(None=Max(A=@n1) {constB=0})
  n2 = TextureSample() {samplerType=SAMPLERTYPE_Masks, texture=/Game/Environments/Sets/Neo/Props/Textures/T_NeoTilted_Spotlight_01_S}
  n3 = VectorParameter() {parameterName=EmissiveColorMultiplier, defaultValue=<rgba 0 4.9533 10 0>}
  n4 = Constant() {r=34.9111}
  n5 = Normalize(VectorInput=RotateAboutAxis(
    NormalizedRotationAxis=Constant3Vector() {constant=<rgba 0 0 1 0>}
    RotationAngle=Multiply(A=@n4, B=@n1) {constA=16}
    PivotPoint=Constant3Vector()
    Position=VectorParameter() {parameterName=Rotation Vector, defaultValue=<rgba 1 0 0 1>}.RGB
   ))
  n6 = Multiply(
   A=Add(A=Time(), B=ScalarParameter() {parameterName=Rotation Timing Offset})
   B=ScalarParameter() {defaultValue=0.0417, parameterName=Rotation Speed}
  )
  n7 = Add(
   A=ScalarParameter() {parameterName=Rotation Angle Offset}
   B=Multiply(A=Sine(None=@n6), B=ScalarParameter() {defaultValue=0.05, parameterName=Rotation Max})
  )
  n8 = MaterialFunctionCall(Rotation Axis=@n5, Rotation Angle=@n7) {materialFunction=&FixRotateAboutAxisNormals}
  n9 = VertexInterpolator(VS=@n8.New Tangent Space Vertex Normal)
  n10 = TextureSample() {samplerType=SAMPLERTYPE_Normal, texture=/Game/Environments/Sets/Neo/Props/Textures/T_NeoTilted_Spotlight_01_N}
  n11 = ActorPositionWS()
  n12 = VectorParameter() {parameterName=Pivot Offset Custom, defaultValue=<rgba 0 0 75 1>}
  n13 = StaticSwitchParameter(
   True=RotateAboutAxis(NormalizedRotationAxis=@n5, RotationAngle=@n7, PivotPoint=Add(A=@n11, B=@n12.RGB), Position=WorldPosition().XYZ)
   False=Constant()
  ) {defaultValue=true, parameterName=Rotate}
  n14 = VertexColor()
  MaterialFunctionCall(MA In=MakeMaterialAttributes!(
    BaseColor=TextureSample() {texture=/Game/Environments/Sets/Neo/Props/Textures/T_NeoTilted_Spotlight_01_D}.RGB
    Metallic=@n2.G
    Specular=@n2.G
    Roughness=@n2.G
    Anisotropy
    EmissiveColor=Multiply(A=@n2.A, B=@n3.RGB)
    Opacity
    OpacityMask
    Normal=MaterialFunctionCall(BaseNormal=@n9.PS, AdditionalNormal=@n10.RGB) {materialFunction=&BlendAngleCorrectedNormals}.Result
    Tangent
    WorldPositionOffset=Multiply(A=@n13, B=@n14.R)
    SubsurfaceColor
    ClearCoat
    ClearCoatRoughness
    AmbientOcclusion
    Refraction
    CustomizedUVs_0
    CustomizedUVs_1
    CustomizedUVs_2
    CustomizedUVs_3=@n8.New Tangent Space Vertex Normal RG
    CustomizedUVs_4=@n8.New Tangent Space Vertex Normal RG
    CustomizedUVs_5
    CustomizedUVs_6
    CustomizedUVs_7
    PixelDepthOffset
    ShadingModel
    Displacement
   )) {materialFunction=&MF_CreativeMoveTool}
  Transform(None=@n12.RGB) {transformSourceType=TRANSFORMSOURCE_Local}
 ]
}
